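# Sound cue for the looping ambient sound of the DiamondCase coin container, classed as a loot ambient loop.
# A mixer layers a quiet DillCoin case loop under a full-level, pitched-down DiamondCase loop.
# An enveloper ramps volume up from silence over the result.
# Distance falloff comes from a shared small item-ambient attenuation asset, while the cue's own concurrency override caps simultaneous instances.
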SoundCue /DragonPopGameplayContent/Assets/Audio/CoinContainer/DiamondCase_Ambient_Loop_Cue
root=Enveloper_0
nodes [
 {
  id=Enveloper_0
  children [Mixer_1]
  properties {pitchCurve {editorCurveData {keys [{time=0}]}}, volumeCurve {editorCurveData {keys [<at 0 0>,<at 1 1>]}}}
 }
 {id=Mixer_1, children [Modulator_2,Modulator_0], properties {inputVolume [0.2,1]}}
 {id=Modulator_0, children [WavePlayer_0], properties {pitchMax=0.6, pitchMin=0.6, volumeMax=1, volumeMin=1}}
 {id=Modulator_2, children [WavePlayer_2], properties {pitchMax=1, pitchMin=1, volumeMax=1, volumeMin=1}}
 {id=WavePlayer_0, properties {bLooping=true, soundWaveAssetPtr=&/DiamondCase_loop}}
 {id=WavePlayer_2, properties {bLooping=true, soundWaveAssetPtr=&/DillCoin_Case_Loop}}
]
properties {
 attenuationSettings=/Game/Sounds/Attenuation/Fort_Item_Ambient_Loop_Mono_Small_Att
 bOverrideConcurrency=true
 maxDistance=700
 priority=50
 soundClassObject=/Game/Sounds/FortSoundClassesAndModes/Fort_Loot_Ambient_Loop
 volumeMultiplier=0.2
 attenuationOverrides {
  bEnableFocusInterpolation=true
  bEnableListenerFocus=true
  distanceAlgorithm=Logarithmic
  falloffDistance=1100
  focusAttackInterpSpeed=20
  focusAzimuth=20
  focusReleaseInterpSpeed=10
  focusVolumeAttenuation=3.5
  lPFFrequencyAtMax=10000
  lPFRadiusMax=1000
  lPFRadiusMin=500
  nonFocusAzimuth=40
  nonFocusVolumeAttenuation=0.8
  nonSpatializedRadiusStart=200
  occlusionInterpolationTime=0.3
  occlusionLowPassFilterFrequency=3000
  occlusionVolumeAttenuation=0.75
  spatializationAlgorithm=SPATIALIZATION_HRTF
  attenuationShapeExtents {x=0}
  customAttenuationCurve {editorCurveData {keys [<at 0 1 RCIM_Cubic>, {arriveTangent=-1, interpMode=RCIM_Cubic, leaveTangent=-1, time=0.5, value=0.1}, <at 1 0 RCIM_Cubic>]}}
 }
 concurrencyOverrides {maxCount=3, voiceStealReleaseTime=0.25}
 concurrencySet [/Game/Sounds/Concurrency_Rules/Loot_Ambient_Conc]
}
