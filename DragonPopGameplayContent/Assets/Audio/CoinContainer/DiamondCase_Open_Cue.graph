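# Sound cue for opening the diamond coin case, played as a positional loot reward sound with item-interact attenuation.
# After a short initial delay, a mixer layers a diamond currency sweetener, a louder coin drop, and a slightly later treasure chest sweetener.
# The chest layer picks randomly from four variants at reduced volume, with randomized upward pitch.
# Overridden concurrency limits it to one instance per owner.
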SoundCue /DragonPopGameplayContent/Assets/Audio/CoinContainer/DiamondCase_Open_Cue
root=Delay_1
nodes [
 {id=Delay_0, children [Modulator_3], properties {delayMax=0.05, delayMin=0.05}}
 {id=Delay_1, children [Mixer_1], properties {delayMax=0.12, delayMin=0.12}}
 {id=Mixer_1, children [Delay_0,WavePlayer_0,WavePlayer_1], properties {inputVolume [0.6,0.6,1.3]}}
 {id=Modulator_3, children [Random_6], properties {pitchMax=1.5, pitchMin=1, volumeMax=0.7, volumeMin=0.7}}
 {
  id=Random_6
  children [WavePlayer_16, WavePlayer_17, WavePlayer_18, WavePlayer_19]
  properties {bShouldExcludeFromBranchCulling=true, weights [1,1,1,1]}
 }
 {id=WavePlayer_0, properties {soundWaveAssetPtr=&/DiamondCurrency_Sweetner}}
 {id=WavePlayer_1, properties {soundWaveAssetPtr=/DragonPopGameplayContent/Gameplay/DiamondCurrency/Audio/DillCoin_Currency_Drop}}
 {id=WavePlayer_16, properties {soundWaveAssetPtr=/Game/Sounds/Foley_Loot/Containers/Treasure_Chest/treasure_chest_sweet_shorter_01}}
 {id=WavePlayer_17, properties {soundWaveAssetPtr=/Game/Sounds/Foley_Loot/Containers/Treasure_Chest/treasure_chest_sweet_shorter_02}}
 {id=WavePlayer_18, properties {soundWaveAssetPtr=/Game/Sounds/Foley_Loot/Containers/Treasure_Chest/treasure_chest_sweet_shorter_03}}
 {id=WavePlayer_19, properties {soundWaveAssetPtr=/Game/Sounds/Foley_Loot/Containers/Treasure_Chest/treasure_chest_sweet_shorter_04}}
]
properties {
 attenuationSettings=/Game/Sounds/Attenuation/Fort_Item_Interact_Att
 bOverrideConcurrency=true
 maxDistance=1850
 soundClassObject=&Fort_Loot_Reward_PSM
 volumeMultiplier=1.1
 concurrencyOverrides {bLimitToOwner=true, maxCount=1}
}
